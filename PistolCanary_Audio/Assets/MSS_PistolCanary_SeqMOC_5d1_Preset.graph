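# One-shot 5.1 MetaSound source for Pistol Canary audio that plays a supplied wave asset on play, using the large POI attenuation and the Fort_Events sound class.
# A 5.1 wave player feeds its front and side pairs into the MSP_AzimuthPannerQuad patch, driven by listener azimuth and MasterVolume.
# Front center and LFE bypass the panner, so they receive neither panning nor MasterVolume gain.
# The player's finish trigger ends the sound.
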MetaSoundSource /PistolCanary_Audio/Assets/MSS_PistolCanary_SeqMOC_5d1_Preset
inputs [
 MasterVolume {type=Float}
 UE.Listener.Orientation.Azimuth {type=Float}
 UE.Source.OnPlay {type=Trigger}
 Wave Asset {type=WaveAsset}
]
interfaces [UE.OutputFormat.5dot1:1.0, UE.Source.OneShot:1.0, UE.Source:1.1]
outputs [
 "UE.OutputFormat.5dot1.Audio:0" {type=Audio}
 "UE.OutputFormat.5dot1.Audio:1" {type=Audio}
 "UE.OutputFormat.5dot1.Audio:2" {type=Audio}
 "UE.OutputFormat.5dot1.Audio:3" {type=Audio}
 "UE.OutputFormat.5dot1.Audio:4" {type=Audio}
 "UE.OutputFormat.5dot1.Audio:5" {type=Audio}
 "UE.OutputFormat.Quad.Audio:0" {type=Audio}
 "UE.OutputFormat.Quad.Audio:1" {type=Audio}
 "UE.OutputFormat.Quad.Audio:4" {type=Audio}
 "UE.OutputFormat.Quad.Audio:5" {type=Audio}
 UE.Source.OneShot.OnFinished {type=Trigger}
]
pages [Default {
  input_defaults {MasterVolume=1, UE.Listener.Orientation.Azimuth=117.031, UE.Source.OnPlay=null, Wave Asset=null}
  nodes [
   InitVariable.Trigger()
   WavePlayer_0 = UE.Wave Player:5dot1(Play=^UE.Source.OnPlay, Wave Asset=^Wave Asset)
   MSP_AzimuthPannerQuad_0 = &MSP_AzimuthPannerQuad(
    Gain=^MasterVolume
    Azimuth=^UE.Listener.Orientation.Azimuth
    In Front Left=@WavePlayer_0.Out Front Left
    In Front Right=@WavePlayer_0.Out Front Right
    In Side Left=@WavePlayer_0.Out Side Left
    In Side Right=@WavePlayer_0.Out Side Right
   )
  ]
  links [
   {dst="@UE.OutputFormat.5dot1.Audio:0", src=MSP_AzimuthPannerQuad_0.Out Front Left}
   {dst="@UE.OutputFormat.5dot1.Audio:1", src=MSP_AzimuthPannerQuad_0.Out Front Right}
   {dst="@UE.OutputFormat.5dot1.Audio:4", src=MSP_AzimuthPannerQuad_0.Out Side Left}
   {dst="@UE.OutputFormat.5dot1.Audio:5", src=MSP_AzimuthPannerQuad_0.Out Side Right}
   {dst="@UE.Source.OneShot.OnFinished", src=WavePlayer_0.On Finished}
   {dst="@UE.OutputFormat.5dot1.Audio:2", src=WavePlayer_0.Out Front Center}
   {dst="@UE.OutputFormat.5dot1.Audio:3", src=WavePlayer_0.Out Low Frequency}
  ]
 }]
properties {
 assetImportData=&self:AssetImportData
 attenuationSettings=/PistolCanary_Audio/Assets/Settings/ATT_PistolCanary_POI_Large
 maxDistance=40000
 numChannels=6
 outputFormat=FiveDotOne
 soundClassObject=&Fort_Events
}
